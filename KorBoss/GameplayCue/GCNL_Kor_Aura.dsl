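; Looping gameplay cue for the Kor boss aura that keeps its particle effect in sync with the boss's cloak state.
; On looping start it caches a spawned FX component, casts the target to the Cherry Smoke Kor boss pawn and binds to its just-cloaked and just-uncloaked dispatchers.
; Those handlers set the bIsCloaked bool parameter on the cached component, leaving the visual response to the effect itself.
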(event EventOnLoopingStart(EitherVFX-preferthisversion)
  (params (MyTarget :type Actor%20Object%20Reference) (Parameters :type Gameplay%20Cue%20Parameters%20Structure%20%28by%20ref%29) (ParticleComponents :type Array%20of%20FXSystem%20Component%20Object%20References) (AudioComponents :type Array%20of%20Audio%20Component%20Object%20References))
  (Variables|Default|SetNiagara%20System%20Cache (Utilities|Array|Get(acopy) ParticleComponents))
  (bind _asbp_player_pawn_cherry_smoke_kor_boss (Utilities|Casting|CastToBP_PlayerPawn_CherrySmoke_KorBoss MyTarget))
  (bind _output_get (Variables|Default|SetAs%20BP%20Player%20Pawn%20Cherry%20Smoke%20Kor%20Boss _asbp_player_pawn_cherry_smoke_kor_boss))
  (Default|BindEventtoKorJustCloaked _output_get (EventDispatchers|CreateEvent :!event Custom|IsCloaked))
  (Default|BindEventtoKorJustUncloaked _output_get (EventDispatchers|CreateEvent :!event Custom|IsUncloaked)))

(event Custom|IsCloaked
  (Effects|Components|ParticleSystem|SetBoolParameter (Variables|Default|GetNiagara%20System%20Cache) "bIsCloaked" true))

(event Custom|IsUncloaked
  (Effects|Components|ParticleSystem|SetBoolParameter (Variables|Default|GetNiagara%20System%20Cache) "bIsCloaked" false))

(fn ConstructionScript)
