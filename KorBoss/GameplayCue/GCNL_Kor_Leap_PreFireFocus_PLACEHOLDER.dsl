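; Placeholder looping gameplay cue for the Kor boss's leap pre-fire focus, with sound and particle effects.
; On loop start it enables the target's tick, runs the parent logic, raises the first audio component and sets the first particle's color index.
; It finds the first KorPawn_Interface actor and feeds its raised location, or the target's if none, to the second particle's TargetEnd.
; That same world-space point becomes the third particle's relative location.
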(event EventOnLoopingStart(EitherVFX-preferthisversion)
  (params (MyTarget :type Actor%20Object%20Reference) (Parameters :type Gameplay%20Cue%20Parameters%20Structure%20%28by%20ref%29) (ParticleComponents :type Array%20of%20FXSystem%20Component%20Object%20References) (AudioComponents :type Array%20of%20Audio%20Component%20Object%20References))
  (bind _returnvalue (Transformation|GetActorLocation MyTarget))
  (Actor|Tick|SetActorTickEnabled MyTarget true)
  (|CallParentFunction MyTarget Parameters ParticleComponents AudioComponents)
  (bind _outactors (Actor|GetAllActorswithInterface MyTarget "/KorBoss/Interface/KorPawn_Interface.KorPawn_Interface_C"))
  (Transformation|SetRelativeLocation (Utilities|Array|Get(acopy) AudioComponents) "0.000000,0.000000,132.000000" false false)
  (Effects|Components|ParticleSystem|SetIntParameter (Utilities|Array|Get(acopy) ParticleComponents) "index_color" 1)
  (Effects|Components|ParticleSystem|SetVectorParameter (Utilities|Array|Get(acopy) ParticleComponents 1) "TargetEnd" (Math|Vector|SelectVector (+ (Transformation|GetActorLocation (Utilities|Array|Get(acopy) _outactors)) "0.000000,0.000000,30.000000") _returnvalue (Utilities|IsValid (Utilities|Array|Get(acopy) _outactors) :!class /Script/Engine.KismetSystemLibrary)))
  (Transformation|SetRelativeLocation (Utilities|Array|Get(acopy) ParticleComponents 2) (Math|Vector|SelectVector (+ (Transformation|GetActorLocation (Utilities|Array|Get(acopy) _outactors)) "0.000000,0.000000,30.000000") _returnvalue (Utilities|IsValid (Utilities|Array|Get(acopy) _outactors) :!class /Script/Engine.KismetSystemLibrary)) false false))

(fn ConstructionScript)
